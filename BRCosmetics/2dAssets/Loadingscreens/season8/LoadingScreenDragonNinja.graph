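# Season 8 Dragon Ninja loading screen widget that cycles three character artworks over an animated beam background with skewed letterbox bars.
# On construct, an Intro animation fades out a dark cover and animates the beams, while a self-restarting Timer animation acts as the swap interval.
# Each swap sets the character dynamic material's Incoming texture and animates its Switch parameter.
# When the swap finishes, that texture becomes Base, TransitionReset plays and ActiveIndex advances with wraparound.
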WidgetBlueprint /BRCosmetics/2dAssets/Loadingscreens/season8/LoadingScreenDragonNinja
animations [
 Intro {
  binding_count=2
  length_seconds=2.1333
  object=&self:Intro
  sequence {
   bindings [
    Image_Fade {
     id=#1
     possessed_class=Image
     tracks [{
       type=MovieSceneFloatTrack
       display_name=Render Opacity
       sections [{type=MovieSceneFloatSection, channels [None {type=MovieSceneScriptingFloatChannel, keys [[0,1],[64,0]]}]}]
      }]
    }
    Image_BG {
     id=#2
     possessed_class=Image
     tracks [{
       type=MovieSceneWidgetMaterialTrack
       sections [{
         type=MovieSceneParameterSection
         channels [
          BeamRadius {type=MovieSceneScriptingFloatChannel, keys [[31,5.4585],[128,1.324]]}
          BeamDensity {type=MovieSceneScriptingFloatChannel, keys [[31,0.0281],[128,1.3377]]}
         ]
        }]
      }]
    }
   ]
   display_rate {numerator=60}
  }
 }
 SwapImage {
  binding_count=1
  length_seconds=1.3738
  object=&self:SwapImage
  sequence {bindings [Image_Character {
     id=#3
     possessed_class=Image
     tracks [{
       type=MovieSceneWidgetMaterialTrack
       sections [{
         type=MovieSceneParameterSection
         start_frame=0
         channels [
          Switch {type=MovieSceneScriptingFloatChannel, keys [[0,1],[27.4763,-0.4189]]}
          BodyRevealOffset {type=MovieSceneScriptingFloatChannel, const [0.0514,1]}
         ]
        }]
      }]
    }]}
 }
 Timer {
  binding_count=1
  length_seconds=2
  object=&self:Timer
  sequence {bindings [Overlay_0 {
     id=#4
     possessed_class=Overlay
     tracks [{
       type=MovieSceneFloatTrack
       display_name=Render Opacity
       sections [{type=MovieSceneFloatSection, end_frame=40, start_frame=0, channels [None {type=MovieSceneScriptingFloatChannel, const [1,2]}]}]
      }]
    }]}
 }
 TransitionReset {
  binding_count=1
  length_seconds=0.5
  object=&self:TransitionReset
  sequence {bindings [Image_Character {
     id=#5
     possessed_class=Image
     tracks [{
       type=MovieSceneWidgetMaterialTrack
       sections [{
         type=MovieSceneParameterSection
         end_frame=10
         start_frame=0
         channels [BodyRevealOffset {type=MovieSceneScriptingFloatChannel, keys [[0,0.0514],[10,-1]]}]
        }]
      }]
    }]}
 }
]
graphs [EventGraph {nodes [
   EventSwapImage() {is=title K2Node_CustomEvent}
   > PlayAnimation(InAnimation=Get SwapImage() -> (SwapImage:Widget Animation Object Reference))
   > SetTextureParameterValue(
    self=Get ImageMID() -> (ImageMID:Material Instance Dynamic Object Reference)
    ParameterName=Incoming
    Value="Get (a copy)"(
     Array:Array of Texture Object References=Get ImageArray() -> (ImageArray:Array of Texture Object References)
     Dimension 1=Get ActiveIndex() -> (ActiveIndex:int=0)
    ) -> (Output:Texture Object Reference)
   )
   n1 = EventTimerComplete() {is=title K2Node_CustomEvent}
   > EventSwapImage() {is=title K2Node_CallFunction}
   > PlayAnimation(InAnimation=Get Timer() -> (Timer:Widget Animation Object Reference))
   n2 = EventSwapFinished() {is=title K2Node_CustomEvent}
   > SetTextureParameterValue(self=Get ImageMID(), ParameterName=Base, Value="Get (a copy)"(Array=Get ImageArray(), Dimension 1=Get ActiveIndex() -> (ActiveIndex=0)))
   > PlayAnimation(InAnimation=Get TransitionReset() -> (TransitionReset:Widget Animation Object Reference))
   n3 = int - int(A=Length(TargetArray:Array of Texture Object References=Get ImageArray()), B=1) {is=title K2Node_CallFunction}
   n4 = int + int(A=Get ActiveIndex() -> (ActiveIndex=0), B=1) -> (ReturnValue=0) {is=title K2Node_CommutativeAssociativeBinaryOperator}
   > Set ActiveIndex(ActiveIndex:int=Select(Option 0:int=@n4, Option 1:int=0, Index="integer > integer"(A=@n4, B=@n3) {is=title K2Node_CallFunction}) -> (ReturnValue:int)) -> (Output_Get:int=0)
   Event Construct()
   > BindToAnimationFinished(Animation=Get Timer(), Delegate=@n1)
   > BindToAnimationFinished(Animation=Get SwapImage(), Delegate=@n2)
   > PlayAnimation(InAnimation=Get Timer())
   > PlayAnimationForward(InAnimation=Get Intro())
   > n5 = GetDynamicMaterial(self=Get Image_Character() -> (Image_Character:Image Object Reference))
   > Set ImageMID(ImageMID:Material Instance Dynamic Object Reference=@n5) -> (Output_Get:Material Instance Dynamic Object Reference)
  ]}]
variables [
 ActiveIndex {type=integer, default=1}
 ImageArray {
  type="array<Texture>"
  default [
   /BRCosmetics/2dAssets/Loadingscreens/season8/DragonNinjaLoading/T_UI_DragonNinja_p1
   /BRCosmetics/2dAssets/Loadingscreens/season8/DragonNinjaLoading/T_UI_DragonNinja_p2
   /BRCosmetics/2dAssets/Loadingscreens/season8/DragonNinjaLoading/T_UI_DragonNinja_p3
  ]
 }
 ImageMID {type=MaterialInstanceDynamic}
]
widgets {
 info {inheritedWidgetCount=0, parentClass=&UserWidget, rootWidgetClass=&Overlay, widgetCount=7}
 widgets [
  Overlay_0 {widgetClassPath=&Overlay, widget=Overlay_0}
  Image_BG {
   widgetClassPath=&Image
   parent=Overlay_0
   widget=Image_BG
   properties {brush {
     bIsDynamicallyLoaded=false
     drawAs=Image
     imageType=NoImage
     mirroring=NoMirror
     resourceName=None
     resourceObject=/BRCosmetics/2dAssets/Loadingscreens/season8/DragonNinjaLoading/M_UI_LoadingDragonNinja_BG_Inst
     tiling=NoTile
     imageSize=<1920 1080>
     margin=<ltrb 0 0 0 0>
     outlineSettings {
      bUseBrushTransparency=false
      roundingType=HalfHeightRadius
      width=0
      color {colorUseRule=UseColor_Specified, specifiedColor=<rgba 0 0 0 0>}
      cornerRadii=<0 0 0 0>
     }
     tintColor {colorUseRule=UseColor_Specified, specifiedColor=<rgba 1 1 1 1>}
     uVRegion {bIsValid=false, max=<0 0>, min=<0 0>}
    }}
   slot {class=OverlaySlot, properties {horizontalAlignment=HAlign_Fill, verticalAlignment=VAlign_Fill}}
  }
  Image_LetterboxTop {
   widgetClassPath=&Image
   parent=Overlay_0
   widget=Image_LetterboxTop
   properties {
    brush {
     bIsDynamicallyLoaded=false
     drawAs=Image
     imageType=NoImage
     mirroring=NoMirror
     resourceName=None
     resourceObject=/BRCosmetics/2dAssets/Loadingscreens/Season7/NeonCatLoading/M_UI_HUD_LetterBoxSkew_Inst
     tiling=NoTile
     imageSize=<32 100.0596>
     margin=<ltrb 0 0 0 0>
     outlineSettings {
      bUseBrushTransparency=false
      roundingType=HalfHeightRadius
      width=0
      color {colorUseRule=UseColor_Specified, specifiedColor=<rgba 0 0 0 0>}
      cornerRadii=<0 0 0 0>
     }
     tintColor {colorUseRule=UseColor_Specified, specifiedColor=<rgba 1 1 1 1>}
     uVRegion {bIsValid=false, max=<0 0>, min=<0 0>}
    }
    colorAndOpacity=<rgba 0.0103 0.0048 0.0037 1>
    renderTransformPivot=<0 0>
   }
   slot {class=OverlaySlot, properties {horizontalAlignment=HAlign_Fill}}
  }
  ScaleBox_0 {
   widgetClassPath=&ScaleBox
   parent=Overlay_0
   widget=ScaleBox_0
   slot {class=OverlaySlot, properties {horizontalAlignment=HAlign_Fill, verticalAlignment=VAlign_Fill}}
  }
  Image_Character {
   widgetClassPath=&Image
   bIsVariable=true
   parent=ScaleBox_0
   widget=Image_Character
   properties {brush {
     bIsDynamicallyLoaded=false
     drawAs=Image
     imageType=NoImage
     mirroring=NoMirror
     resourceName=None
     resourceObject=/BRCosmetics/2dAssets/Loadingscreens/season8/DragonNinjaLoading/M_UI_DragonNinjaLoading_Character_Base_p1
     tiling=NoTile
     imageSize=<854 1080>
     margin=<ltrb 0 0 0 0>
     outlineSettings {
      bUseBrushTransparency=false
      roundingType=HalfHeightRadius
      width=0
      color {colorUseRule=UseColor_Specified, specifiedColor=<rgba 0 0 0 0>}
      cornerRadii=<0 0 0 0>
     }
     tintColor {colorUseRule=UseColor_Specified, specifiedColor=<rgba 1 1 1 1>}
     uVRegion {bIsValid=false, max=<0 0>, min=<0 0>}
    }}
   slot {class=ScaleBoxSlot, properties {verticalAlignment=VAlign_Bottom}}
  }
  Image_LetterboxBottom {
   widgetClassPath=&Image
   parent=Overlay_0
   widget=Image_LetterboxBottom
   properties {
    brush {
     bIsDynamicallyLoaded=false
     drawAs=Image
     imageType=NoImage
     mirroring=NoMirror
     resourceName=None
     resourceObject=&M_UI_HUD_LetterBoxSkew_Inst
     tiling=NoTile
     imageSize=<32 122.1887>
     margin=<ltrb 0 0 0 0>
     outlineSettings {
      bUseBrushTransparency=false
      roundingType=HalfHeightRadius
      width=0
      color {colorUseRule=UseColor_Specified, specifiedColor=<rgba 0 0 0 0>}
      cornerRadii=<0 0 0 0>
     }
     tintColor {colorUseRule=UseColor_Specified, specifiedColor=<rgba 1 1 1 1>}
     uVRegion {bIsValid=false, max=<0 0>, min=<0 0>}
    }
    colorAndOpacity=<rgba 0.0103 0.0048 0.0037 1>
    renderTransform {angle=0, scale=<-1 -1>, shear=<0 0>, translation=<0 0>}
   }
   slot {class=OverlaySlot, properties {horizontalAlignment=HAlign_Fill, verticalAlignment=VAlign_Bottom}}
  }
  Image_Fade {
   widgetClassPath=&Image
   bIsVariable=true
   parent=Overlay_0
   widget=Image_Fade
   properties {renderOpacity=0, colorAndOpacity=<rgba 0.0052 0.0052 0.0056 1>}
   slot {class=OverlaySlot, properties {horizontalAlignment=HAlign_Fill, verticalAlignment=VAlign_Fill}}
  }
 ]
}
